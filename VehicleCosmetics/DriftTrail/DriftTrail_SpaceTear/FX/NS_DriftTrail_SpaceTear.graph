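# Niagara system for the SpaceTear vehicle drift trail cosmetic, combining rear-trailing ribbons with mesh wraps on the wheels.
# FillTrail001 spawns two skeletal-positioned ribbons with five renderers, switching spawn rate, lifetime, offset and velocity by GameModeID, with curve-driven glow sizes partly scaled by ColorVariantID.
# The WheelWrap emitter bursts four mesh particles pinned to wheel bones and scaled per wheel.
# Both apply painted-variant color overrides and per-wheel tire FX masking.
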NiagaraSystem /VehicleCosmetics/DriftTrail/DriftTrail_SpaceTear/FX/NS_DriftTrail_SpaceTear
dependencies {
 usedDataInterfaces [&NiagaraDataInterfaceColorCurve, &NiagaraDataInterfaceSkeletalMesh]
 usedDynamicInputs []
 usedModules [
  &self:SystemSpawnScript.SystemScriptSource.SystemScriptGraph.NiagaraNodeAssignment_0.SetVariables_#1
  &FN_UnpackVehicleFlags
  /Game/Effects/Niagara/Modules/Vehicle/FN_UpdateVehicleAttribs
  &SystemState
  /Game/Effects/Niagara/Modules/Vehicle/FN_VehicleFlags_Cheatcodes
 ]
}
emitters {
 FillTrail001 {
  emitterUpdateScript [
   &EmitterState(Life Cycle Mode=::NewEnumerator1, Loop Duration=1, UseLoopDelay=false)
   &VC_RateQualityHandler()
   &VC_SpawnRibbon_v1(
    Number of Ribbons=2
    Spawn Time=-1
    SpawnRate=&MergeHelper_FloatSwitch(0=15, 1=15, 2=20, 3=25, Index=$User.GameModeID, Num=::NewEnumerator2)
   )
   /SetVariables(
    Emitter.Panner Speed:NiagaraFloat=$User.Speed Ribbon
    Emitter.RGB Emiss Stars:LinearColor=&MergeHelper_ColorSwitch(0=<rgba 3 0 0 0>, 1=<rgba 12 0 0 0>, 2=<rgba 12 0 0 0>, 3=<rgba 3 0 0 0>, Index=$User.GameModeID, Num=::NewEnumerator2)
   ) {isSetParameters=true}
  ]
  particleSpawnScript [
   &VC_PaintedVariantsOverride(
    Black=<rgba 0.1163 0.1163 0.1163 1>
    BurntSienna=<rgba 0.8177 0.2278 0.0696 1>
    Crimson=<rgba 0.8316 0.091 0.091 1>
    DefaultColor=<rgba 0 0.0093 2 1>
    Lime=<rgba 0.5 1 0 1>
    Orange=<rgba 1 0 0.0372 1>
    Pink=<rgba 0.849 0.1017 0.5772 1>
    Purple=<rgba 0.2789 0 0.6795 0>
    R1 Black:LinearColor=<rgba 0.1163 0.1163 0.1163 1>
    R1 BurntSienna:LinearColor=<rgba 0.474 0.2576 0.0312 1>
    R1 Crimson:LinearColor=<rgba 1 0 0 1>
    R1 DefaultColor=<rgba 1.9265 0.7664 2 1>
    R1 Lime:LinearColor=<rgba 0.5 1 0 1>
    R1 Orange:LinearColor=<rgba 1 0.1947 0 1>
    R1 Pink:LinearColor=<rgba 1 0 0.6364 1>
    R1 Purple:LinearColor=<rgba 0.2789 0 0.6795 0>
    R1 SkyBlue:LinearColor=<rgba 0.2994 0 3 1>
    R2 BurntSienna:LinearColor=<rgba 0.03 0 0 1>
    R2 Cobalt:LinearColor=<rgba 0 0 0.03 1>
    R2 Crimson:LinearColor=<rgba 0.015 0 0.0003 1>
    R2 DefaultColor:LinearColor=<rgba 0 0 0 1>
    R2 ForestGreen:LinearColor=<rgba 0 0.01 0 1>
    R2 Grey:LinearColor=<rgba 0.2031 0.2031 0.2031 0>
    R2 Lime:LinearColor=<rgba 0.005 0.01 0 1>
    R2 Orange:LinearColor=<rgba 0 0 0 1>
    R2 Pink:LinearColor=<rgba 0.01 0.0012 0.0068 1>
    R2 Purple:LinearColor=<rgba 0.2789 0 0.6795 0>
    R2 Saffron:LinearColor=<rgba 0.01 0.0077 0 1>
    R2 SkyBlue:LinearColor=<rgba 0 0.0067 0.01 1>
    R2 TitaniumWhite:LinearColor=<rgba 0.01 0.01 0.01 1>
    R3 BurntSienna:LinearColor=<rgba 0.7969 0.2036 0.1026 1>
    R3 Cobalt:LinearColor=<rgba 0 0.046 0.5434 1>
    R3 Crimson:LinearColor=<rgba 0.8594 0 0 1>
    R3 DefaultColor:LinearColor=<rgba 0.0355 0 0.4983 1>
    R3 ForestGreen:LinearColor=<rgba 0 0.4219 0.0213 0.3507>
    R3 Grey:LinearColor=<rgba 0.2031 0.2031 0.2031 0>
    R3 Lime:LinearColor=<rgba 0.5326 0.7656 0 1>
    R3 Orange:LinearColor=<rgba 2 0.6976 0 1>
    R3 Pink:LinearColor=<rgba 0.6441 0.1979 0.4819 1>
    R3 Purple:LinearColor=<rgba 0.2789 0 0.6795 0>
    R3 Saffron:LinearColor=<rgba 1.2 1.12 0 1>
    R3 SkyBlue:LinearColor=<rgba 0 0.9981 1.5 1>
    R4 BurntSienna:LinearColor=<rgba 0.6302 0.2712 0.0973 1>
    R4 Cobalt:LinearColor=<rgba 0 0.0185 1 1>
    R4 Crimson:LinearColor=<rgba 1 0 0 1>
    R4 DefaultColor:LinearColor=<rgba 0.5953 0 1.5 1>
    R4 Lime:LinearColor=<rgba 0.7646 1 0 1>
    R4 Purple:LinearColor=<rgba 0.2789 0 0.6795 0>
    R4 SkyBlue:LinearColor=<rgba 0 0.6654 1 1>
    SkyBlue=<rgba 0 1.9962 3 1>
    Use R1 Painted Overrides=true
    Use R2 Painted Overrides=true
    Use R3 Painted Overrides=true
    Use R4 Painted Overrides=true
   )
   &VC_InitializeRibbon_v1(
    Color=$Particles.Color
    Custom Offset=true
    Enable Skeletal Position=true
    Lifetime=&MergeHelper_FloatSwitch(0=1.25, 1=1.25, 2=1.25, 3=1.5, Index=$User.GameModeID, Num=::NewEnumerator2)
    Local Vector Offset=&MergeHelper_VectorSwitch(0=<40 0 0>, 1=<100 0 0>, 2=<100 0 0>, 3=<80 0 0>, Index=$User.GameModeID, Num=::NewEnumerator2)
    Offset Amount=1
    Ribbon Width=60
   )
   &AddVelocity(Velocity=&MergeHelper_VectorSwitch(0=<-1200 0 0>, 1=<-2000 0 0>, 2=<-2000 0 0>, 3=<-1500 0 0>, Index=$User.GameModeID, Num=::NewEnumerator2)) {module=AddVelocity001}
  ]
  particleUpdateScript [
   &FN_DisableTireFX(Mask Index Attribute:NiagaraInt32=$Particles.SpawnGroup, Mask Index FL=2, Mask Index FR=3, Mask Index RL=0, Mask Index RR=1)
   &InheritSourceMovement(
    Applied Position Scale=&VectorFromFloat(Value=&FloatFromCurve(
      FloatCurve=DataInterface() {propertyValues {Curve {
         defaultValue=1
         keys [
          {arriveTangentWeight=0.3333, interpMode=RCIM_Cubic, leaveTangentWeight=0.3333, tangentMode=RCTM_User, time=0, value=1}
          {
           arriveTangent=-3.2426
           arriveTangentWeight=1.1311
           interpMode=RCIM_Cubic
           leaveTangent=-3.2426
           leaveTangentWeight=1.1311
           tangentMode=RCTM_User
           time=1
           value=0
          }
         ]
        }}}
      Scale Curve=1.1
     ))
    ApplyForce=false
    ApplyPosition=true
    ApplyPositionFalloff=false
   )
   &ParticleState()
   &SolveForcesAndVelocity(Clamp Velocity=false, Limit Acceleration=false)
   &VC_UpdateRibbon_v1(Enable Skeletal Position=true)
   /SetVariables(
    "Particles.GB Pink ":LinearColor=&MakeCustomLinearColorFromBool(
     Bool=$User.bIsPainted
     True Color=$Particles.R4 Color
     False Color=&ColorFromCurve(ColorCurve=DataInterface() {propertyValues {
        AlphaCurve {keys [<at 0 1 RCIM_Cubic>,<at 0.5131 0.95>,<at 1 1 RCIM_Cubic>]}
        BlueCurve {keys [<at 0 1 RCIM_Cubic>,<at 0.3685 2>,<at 0.7556 0.3467 RCIM_Cubic>]}
        GreenCurve {keys [<at 0 0 RCIM_Cubic>,<at 0.3685 0.0868>,<at 0.7556 0 RCIM_Cubic>]}
        RedCurve {keys [<at 0 0.4411 RCIM_Cubic>,<at 0.3685 0.7883>,<at 0.7556 0.1529 RCIM_Cubic>]}
       }})
    )
    Particles.Size RIbbon Glow:NiagaraFloat=&FloatFromCurve(
     FloatCurve=$User.Ribbon General Ribbon
     Scale Curve=&SelectFloatFromArray(
      Direct Array Index=$User.ColorVariantID
      Float Selection Array=DataInterface() {propertyValues {FloatData [35,40,40,56,40,40,35,40,30,40,30,40,40,40,40]}}
     )
    )
    Particles.Size Ribbon Secon Glow:NiagaraFloat=&FloatFromCurve(
     FloatCurve=$User.Ribbon General Ribbon
     Scale Curve=&SelectFloatFromArray(
      Direct Array Index=$User.ColorVariantID
      Float Selection Array=DataInterface() {propertyValues {FloatData [48,50,50,50,80,70,100,50,65,50,50,50,50,60,65]}}
     )
    )
    Particles.Size Ribbon Tree Glow:NiagaraFloat=&FloatFromCurve(
     FloatCurve=DataInterface() {propertyValues {
       bExposeCurve=true
       Curve {
        defaultValue=0.5
        keys [
         {arriveTangent=0.0498, interpMode=RCIM_Cubic, leaveTangent=0.0498, tangentMode=RCTM_User, time=0.2197, value=1.15}
         {interpMode=RCIM_Cubic, tangentMode=RCTM_User, time=0.35, value=2}
         {arriveTangent=-3.3993, interpMode=RCIM_Cubic, leaveTangent=-3.4826, tangentMode=RCTM_Break, time=0.6, value=0.65}
         {arriveTangent=-1.1667, tangentMode=RCTM_User, time=1, value=0}
        ]
       }
      }}
     Scale Curve=75
    )
    Particles.Size Stars 01:NiagaraFloat=&FloatFromCurve(FloatCurve=$User.Ribbon General Ribbon, Scale Curve=25)
   ) {isSetParameters=true}
  ]
  rendererClasses [&NiagaraRibbonRendererProperties]
  renderers [
   {rendererClass=&NiagaraRibbonRendererProperties}
   {rendererClass=&NiagaraRibbonRendererProperties, rendererIndex=1}
   {rendererClass=&NiagaraRibbonRendererProperties, rendererIndex=2}
   {rendererClass=&NiagaraRibbonRendererProperties, rendererIndex=3}
   {rendererClass=&NiagaraRibbonRendererProperties, rendererIndex=4}
  ]
 }
 NE_VC_DriftTrail_WheelWrap_Parent001 {
  emitterUpdateScript [
   &EmitterState(
    Inactive Response=::NewEnumerator1
    Life Cycle Mode=::NewEnumerator1
    Loop Behavior=::NewEnumerator1
    Loop Duration Mode=::NewEnumerator1
    UseLoopDelay=false
   )
   &SpawnBurst_Instantaneous(Spawn Count=4, Use Spawn Probability=false, UseLoopCountLimit=false)
  ]
  particleSpawnScript [
   &VC_PaintedVariantsOverride(
    Black=<rgba 0.08 0.08 0.08 5>
    BurntSienna=<rgba 0.6 0.2728 0.1684 1>
    Cobalt=<rgba 0.2717 0.4348 1.25 1>
    Crimson=<rgba 1.5 0.105 0.1428 2>
    DefaultColor=<rgba 0 0.0046 1 1>
    ForestGreen=<rgba 0.5143 0.7899 0.5282 1>
    Grey=<rgba 0.35 0.35 0.35 1>
    Lime=<rgba 0.7328 1 0.1094 1>
    Orange=<rgba 0.89 0.3104 0 1>
    Pink=<rgba 0.8837 0.3977 0.7069 1>
    Purple=<rgba 0.2873 0 0.7 1>
    R1 Black:LinearColor=<rgba 0.15 0.15 0.15 1>
    R1 BurntSienna:LinearColor=<rgba 0.6198 0.3018 0.1715 1>
    R1 Cobalt:LinearColor=<rgba 0.3 0.4167 1 1>
    R1 Crimson:LinearColor=<rgba 1 0.3 0.3 0.7>
    R1 DefaultColor=<rgba 0.9633 0.3832 1 1>
    R1 ForestGreen:LinearColor=<rgba 0 0.5816 0.0294 1>
    R1 Grey:LinearColor=<rgba 0.6 0.6 0.6 0.5>
    R1 Lime:LinearColor=<rgba 0.9312 1.25 0.1875 0.3>
    R1 Orange:LinearColor=<rgba 0.8 0.3218 0.0656 1>
    R1 Pink:LinearColor=<rgba 0.7965 0.3172 0.7824 1>
    R1 Purple:LinearColor=<rgba 0.34 0.17 0.68 0.8>
    R1 Saffron:LinearColor=<rgba 0.9 0.7635 0.315 0.5>
    R1 SkyBlue:LinearColor=<rgba 0.3354 0.6702 0.8385 0.8>
    R1 TitaniumWhite:LinearColor=<rgba 0.6441 0.6441 0.6441 0.8>
    R2 Black:LinearColor=<rgba 0.001 0.001 0.001 10>
    R2 BurntSienna:LinearColor=<rgba 0 0 0 1>
    R2 Cobalt:LinearColor=<rgba 0 0.0002 0.01 1>
    R2 Crimson:LinearColor=<rgba 0 0 0 1>
    R2 DefaultColor:LinearColor=<rgba 0 0 0 1>
    R2 ForestGreen:LinearColor=<rgba 0 0 0 1>
    R2 Grey:LinearColor=<rgba 0.2031 0.2031 0.203 0>
    R2 Lime:LinearColor=<rgba 0.0005 0.001 0 1>
    R2 Orange:LinearColor=<rgba 0 0 0 1>
    R2 Pink:LinearColor=<rgba 0 0 0 1>
    R2 Purple:LinearColor=<rgba 0.2789 0 0.6795 0>
    R2 Saffron:LinearColor=<rgba 0.001 0.0008 0 1>
    R2 SkyBlue:LinearColor=<rgba 0 0 0 1>
    R2 TitaniumWhite:LinearColor=<rgba 0 0 0 1>
    R3 BurntSienna:LinearColor=<rgba 0.6 0.1853 0.074 1>
    R3 Cobalt:LinearColor=<rgba 0 0.0241 1.3 1>
    R3 Crimson:LinearColor=<rgba 1 0 0.0271 1>
    R3 DefaultColor:LinearColor=<rgba 0.2391 0.0507 0.7483 1>
    R3 ForestGreen:LinearColor=<rgba 0 0.8 0.0404 1>
    R3 Lime:LinearColor=<rgba 0.65 1.3 0 1>
    R3 Orange:LinearColor=<rgba 2 0.6976 0 1>
    R3 Pink:LinearColor=<rgba 1.5 0.2682 1.0521 1>
    R3 Saffron:LinearColor=<rgba 1.5 1.15 0 1>
    R3 SkyBlue:LinearColor=<rgba 0 1.3308 2 1>
    R3 TitaniumWhite:LinearColor=<rgba 1.2 1.2 1.2 1>
    Saffron=<rgba 0.6684 0.5124 0 1>
    SkyBlue=<rgba 0.5 0.8327 1 1>
    TitaniumWhite=<rgba 0.6 0.6 0.6 1>
    Use R1 Painted Overrides=true
    Use R2 Painted Overrides=true
    Use R3 Painted Overrides=true
   )
   &InitializeParticle(
    Color=$Particles.Color
    Mesh Scale Mode=::NewEnumerator4
    Sprite Rotation Mode=::NewEnumerator2
    Sprite Size Mode=::NewEnumerator3
    UsePositionOffset=false
    Write Material Random=true
   )
   &PartitionParticles()
   &SkeletalMeshLocation(
    Bone / Socket Index=&Add_Integer(A=$Particles.PartitionParticles.Partition, B=4)
    Bone / Socket Sampling Mode=::NewEnumerator5
    Interpolate Mesh Sampling=false
    MaskSpawnGroup=false
    Mesh Sampling Type=::NewEnumerator6
    Sampled Position Offset=&AddVectorToPosition(Position=<0 0 0>, Vector=$User.offset)
    Skeletal Mesh=$System.SkeletalMesh
   )
  ]
  particleUpdateScript [
   &FN_DisableTireFX(
    Mask Index Attribute:NiagaraInt32=$Particles.PartitionParticles.Partition
    Mask Index FL=2
    Mask Index FR=3
    Mask Index RL=0
    Mask Index RR=1
   )
   &VC_WheelWrapScale(StandardRadius=49, Wheel Index Attribute:NiagaraInt32=&Add_Integer(A=$Particles.PartitionParticles.Partition, B=2))
   &ParticleState(Kill Particles When Lifetime Has Elapsed=false)
   &UpdateMeshOrientation(Rotation Coordinate Space=::NewEnumerator2, Rotation Vector=<0 1 0>) {enabled=false}
   &SkeletalMeshLocation(
    Bone / Socket Index=&Add_Integer(A=$Particles.PartitionParticles.Partition, B=4)
    Bone / Socket Sampling Mode=::NewEnumerator5
    MaskSpawnGroup=false
    Mesh Sampling Type=::NewEnumerator6
    Skeletal Mesh=$System.SkeletalMesh
   ) {module=SkeletalMeshLocation001}
   &SolveForcesAndVelocity(Clamp Velocity=false, Limit Acceleration=false)
  ]
  rendererClasses [&NiagaraMeshRendererProperties]
  renderers [
   {rendererClass=&NiagaraMeshRendererProperties, rendererIndex=0}
   {rendererClass=&NiagaraMeshRendererProperties, rendererIndex=1}
   {rendererClass=&NiagaraMeshRendererProperties, rendererIndex=2}
   {rendererClass=&NiagaraMeshRendererProperties, rendererIndex=3}
  ]
 }
}
summary {
 systemName=NS_DriftTrail_SpaceTear
 emitters [
  {emitterName=NE_VC_DriftTrail_WheelWrap_Parent001, rendererClasses [&NiagaraMeshRendererProperties]}
  {emitterName=FillTrail001, rendererClasses [&NiagaraRibbonRendererProperties]}
 ]
 userVariables [
  User.bIsFrontendPreview {type=NiagaraBool}
  User.bIsPainted {type=NiagaraBool}
  User.ColorVariantID {type=NiagaraInt32}
  User.GameModeID {type=NiagaraInt32, defaultValue=3}
  User.Index_Lifetime {type=NiagaraInt32}
  User.Index_SpawnRate {type=NiagaraInt32}
  User.offset {type=Vector3f, description="offset from wheels(set through cosmetic template)", defaultValue {struct=Vector3f, value=<0 0 0>}}
  User.PaintedColor {type=LinearColor, description="", defaultValue {struct=LinearColor, value=<rgba 1 1 1 1>}}
  User.PaintedScalar {type=NiagaraFloat, defaultValue=1, description=""}
  User.Ribbon General Ribbon {
   type=NiagaraDataInterfaceCurve
   description=""
   defaultValue {
    struct=NiagaraExt_VariableValue_DataInterface
    value {
     dataInterface=&self:NiagaraDataInterfaceCurve_0
     dataInterfaceClass=&NiagaraDataInterfaceCurve
    }
   }
  }
  User.Speed Ribbon {type=NiagaraFloat, defaultValue=-2, description=""}
  User.VehicleFlags {type=NiagaraInt32, defaultValue=806289408}
  User.Wheel_Radius_Array {
   type=NiagaraDataInterfaceArrayFloat
   defaultValue {
    struct=NiagaraExt_VariableValue_DataInterface
    value {
     dataInterface=&self:NiagaraDataInterfaceArrayFloat_1
     dataInterfaceClass=&NiagaraDataInterfaceArrayFloat
    }
   }
  }
  User.Wheel_Width_Array {
   type=NiagaraDataInterfaceArrayFloat
   defaultValue {
    struct=NiagaraExt_VariableValue_DataInterface
    value {
     dataInterface=&self:NiagaraDataInterfaceArrayFloat_0
     dataInterfaceClass=&NiagaraDataInterfaceArrayFloat
    }
   }
  }
  User.WheelTouchingArray {
   type=NiagaraDataInterfaceArrayFloat
   defaultValue {
    struct=NiagaraExt_VariableValue_DataInterface
    value {
     dataInterface=&self:NiagaraDataInterfaceArrayFloat_2
     dataInterfaceClass=&NiagaraDataInterfaceArrayFloat
    }
   }
  }
 ]
}
